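# Sound cue for a blimp POI ambient spot sound of distant metal fire-escape noise.
# A Random node picks one of six equally weighted waves, borrowed from the Hollow Base cave spot set, and feeds a Modulator at the root.
# It plays in the outdoor ambient sound class at reduced volume, using a shared distant-creak one-shot attenuation asset, with bus and submix sends disabled.
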SoundCue /Game/Sounds/Artemis/POI/Blimps/Spots/Spot_Blimp_Distant_Metal_Fire_Escape_Cue
root=Modulator_0
nodes [
 {id=Modulator_0, children [Random_0]}
 {id=Random_0, children [WavePlayer_0,WavePlayer_1,WavePlayer_2,WavePlayer_3,WavePlayer_4,WavePlayer_5], properties {weights [1,1,1,1,1,1]}}
 {id=WavePlayer_0, properties {soundWaveAssetPtr=/Game/Sounds/Artemis/POI/Hollow_Base/Spots/Spot_Cave_Distant_Metal_Fire_Escape_1}}
 {id=WavePlayer_1, properties {soundWaveAssetPtr=/Game/Sounds/Artemis/POI/Hollow_Base/Spots/Spot_Cave_Distant_Metal_Fire_Escape_2}}
 {id=WavePlayer_2, properties {soundWaveAssetPtr=/Game/Sounds/Artemis/POI/Hollow_Base/Spots/Spot_Cave_Distant_Metal_Fire_Escape_3}}
 {id=WavePlayer_3, properties {soundWaveAssetPtr=/Game/Sounds/Artemis/POI/Hollow_Base/Spots/Spot_Cave_Distant_Metal_Fire_Escape_4}}
 {id=WavePlayer_4, properties {soundWaveAssetPtr=/Game/Sounds/Artemis/POI/Hollow_Base/Spots/Spot_Cave_Distant_Metal_Fire_Escape_5}}
 {id=WavePlayer_5, properties {soundWaveAssetPtr=/Game/Sounds/Artemis/POI/Hollow_Base/Spots/Spot_Cave_Distant_Metal_Fire_Escape_6}}
]
properties {
 attenuationSettings=/Game/Sounds/Attenuation/Fort_TileAmbient_OneShot_Distant_Creak_Att
 bEnableBusSends=false
 bEnableSubmixSends=false
 maxDistance=8200
 soundClassObject=&Fort_Ambient_Outdoors_Athena
 volumeMultiplier=0.752
 attenuationOverrides {bAttenuate=false}
}
