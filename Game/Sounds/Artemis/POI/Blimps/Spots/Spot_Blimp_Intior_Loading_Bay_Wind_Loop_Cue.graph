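# Looping spot ambience cue for wind in a blimp's interior loading bay, part of the Artemis POI sounds.
# A looping wave player feeds an Enveloper whose volume curve ramps up from silence at the start.
# An interior/exterior wind transition attenuation asset shapes it spatially, and it plays at reduced volume under the gadgets sound class.
# Overridden concurrency caps simultaneous instances with a timed voice-steal release.
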SoundCue /Game/Sounds/Artemis/POI/Blimps/Spots/Spot_Blimp_Intior_Loading_Bay_Wind_Loop_Cue
root=Enveloper_0
nodes [
 {
  id=Enveloper_0
  children [WavePlayer_1]
  properties {pitchCurve {editorCurveData {keys [{time=0}]}}, volumeCurve {editorCurveData {keys [<at 0 0>,<at 1 1>]}}}
 }
 {
  id=WavePlayer_1
  properties {bLooping=true, soundWaveAssetPtr=&/Spot_Blimp_Interior_Loading_Bay_Wind_Loop}
 }
]
properties {
 attenuationSettings=&/Spot_Blimp_IntExt_Wind_Transiton_Att
 bEnableBusSends=false
 bOverrideConcurrency=true
 maxDistance=1700
 soundClassObject=&Fort_Gadgets
 volumeMultiplier=0.5
 concurrencyOverrides {maxCount=2, voiceStealReleaseTime=2}
 concurrencySet [/Game/Sounds/Concurrency_Rules/Athena_Prop]
 preEffectBusSends [{}]
 soundSubmixSends [{minSendLevel=0, sendLevel=0, soundSubmix=&Fortnite_Reverb_Distant}]
}
