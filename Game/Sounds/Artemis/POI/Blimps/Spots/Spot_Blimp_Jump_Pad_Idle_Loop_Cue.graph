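# Looping ambient sound cue for an idle jump pad at the Artemis blimp POI.
# A single looping wave player feeds a mono jump pad ambience straight to the output, boosted in volume and routed to the Fort_Gadgets sound class.
# A dedicated attenuation asset handles spatial falloff, and an overridden concurrency rule caps simultaneous voices at two with a timed steal release.
# Bus and submix sends are disabled, leaving the listed distant reverb send inactive.
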SoundCue /Game/Sounds/Artemis/POI/Blimps/Spots/Spot_Blimp_Jump_Pad_Idle_Loop_Cue
root=WavePlayer_0
nodes [{id=WavePlayer_0, properties {bLooping=true, soundWaveAssetPtr=&/Blimp_Jump_Pad_Ambient_Loop_Mono}}]
properties {
 attenuationSettings=/Game/Sounds/Artemis/POI/Blimps/Spot_Blimp_Jump_Pad_Idle_Att
 bEnableBusSends=false
 bEnableSubmixSends=false
 bOverrideConcurrency=true
 maxDistance=4000
 soundClassObject=&Fort_Gadgets
 volumeMultiplier=5
 concurrencyOverrides {maxCount=2, voiceStealReleaseTime=2}
 concurrencySet [/Game/Sounds/Concurrency_Rules/Athena_Prop]
 preEffectBusSends [{}]
 soundSubmixSends [{minSendLevel=0, sendLevel=0, soundSubmix=&Fortnite_Reverb_Distant}]
}
