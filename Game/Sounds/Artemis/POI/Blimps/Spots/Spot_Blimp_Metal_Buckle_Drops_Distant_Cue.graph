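# Sound cue for a distant metal buckle drop, a spot one-shot ambient sound for the Artemis blimp POI.
# A Random node picks one of seven equally weighted stereo waves, and a Modulator varies its pitch on each play.
# It uses distant-creak attenuation and the indoor ambient one-shot sound class, sends to a distant reverb submix, and overrides concurrency to two instances.
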SoundCue /Game/Sounds/Artemis/POI/Blimps/Spots/Spot_Blimp_Metal_Buckle_Drops_Distant_Cue
root=Modulator_0
nodes [
 {id=Modulator_0, children [Random_1], properties {pitchMin=0.9}}
 {
  id=Random_1
  children [WavePlayer_9, WavePlayer_11, WavePlayer_12, WavePlayer_13, WavePlayer_14, WavePlayer_15, WavePlayer_16]
  properties {weights [1,1,1,1,1,1,1]}
 }
 {id=WavePlayer_11, properties {soundWaveAssetPtr=&/Spot_Blimp_Metal_Buckle_Drops_Distant_04_Stereo}}
 {id=WavePlayer_12, properties {soundWaveAssetPtr=&/Spot_Blimp_Metal_Buckle_Drops_Distant_05_Stereo}}
 {id=WavePlayer_13, properties {soundWaveAssetPtr=&/Spot_Blimp_Metal_Buckle_Drops_Distant_06_Stereo}}
 {id=WavePlayer_14, properties {soundWaveAssetPtr=&/Spot_Blimp_Metal_Buckle_Drops_Distant_07_Stereo}}
 {id=WavePlayer_15, properties {soundWaveAssetPtr=&/Spot_Blimp_Metal_Buckle_Drops_Distant_08_Stereo}}
 {id=WavePlayer_16, properties {soundWaveAssetPtr=&/Spot_Blimp_Metal_Buckle_Drops_Distant_09_Stereo}}
 {id=WavePlayer_9, properties {soundWaveAssetPtr=&/Spot_Blimp_Metal_Buckle_Drops_Distant_02_Stereo}}
]
properties {
 attenuationSettings=/Game/Sounds/Attenuation/Fort_TileAmbient_OneShot_Distant_Creak_Att
 bOverrideConcurrency=true
 maxDistance=8200
 soundClassObject=/Game/Sounds/FortSoundClassesAndModes/Fort_Ambient_OneShots_Indoors
 volumeMultiplier=0.652
 attenuationOverrides {falloffDistance=6000, stereoSpread=10000, attenuationShapeExtents {x=2000}}
 concurrencyOverrides {maxCount=2}
 soundSubmixSends [{minSendLevel=0.4, soundSubmix=&Fortnite_Reverb_Distant}]
}
