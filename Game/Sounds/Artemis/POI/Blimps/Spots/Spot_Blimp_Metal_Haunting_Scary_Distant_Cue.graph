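# Sound cue for distant, eerie metal one-shot spot ambience at the Artemis blimp POI, classed as Athena outdoor ambient sound.
# A Random node picks one of five equally weighted stereo waves, feeding a Modulator that varies pitch with a lowered minimum.
# A less-positional logarithmic one-shot attenuation asset is used, with overrides for falloff, shape extents and a very wide stereo spread.
# Volume is boosted, and bus and submix sends are disabled.
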SoundCue /Game/Sounds/Artemis/POI/Blimps/Spots/Spot_Blimp_Metal_Haunting_Scary_Distant_Cue
root=Modulator_0
nodes [
 {id=Modulator_0, children [Random_1], properties {pitchMin=0.9}}
 {id=Random_1, children [WavePlayer_8, WavePlayer_9, WavePlayer_10, WavePlayer_11, WavePlayer_12], properties {weights [1,1,1,1,1]}}
 {id=WavePlayer_10, properties {soundWaveAssetPtr=&/Spot_Blimp_Metal_Haunting_Scary_Distant_03_Stereo}}
 {id=WavePlayer_11, properties {soundWaveAssetPtr=&/Spot_Blimp_Metal_Haunting_Scary_Distant_04_Stereo}}
 {id=WavePlayer_12, properties {soundWaveAssetPtr=&/Spot_Blimp_Metal_Haunting_Scary_Distant_05_Stereo}}
 {id=WavePlayer_8, properties {soundWaveAssetPtr=&/Spot_Blimp_Metal_Haunting_Scary_Distant_01_Stereo}}
 {id=WavePlayer_9, properties {soundWaveAssetPtr=&/Spot_Blimp_Metal_Haunting_Scary_Distant_02_Stereo}}
]
properties {
 attenuationSettings=/Game/Sounds/Attenuation/Fort_TileAmbient_OneShot_LessPositional_Log_Att
 bEnableBusSends=false
 bEnableSubmixSends=false
 maxDistance=6200
 soundClassObject=&Fort_Ambient_Outdoors_Athena
 volumeMultiplier=2.506
 attenuationOverrides {falloffDistance=6000, stereoSpread=10000, attenuationShapeExtents {x=2000}}
}
