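# Sound cue for distant metal rubbing spot ambience at the blimp POI, routed to the outdoor Athena ambient sound class.
# A Random node picks one of eight equally weighted metal rub waves and feeds a Modulator that varies pitch.
# Playback uses a less-positional one-shot ambient attenuation asset with a long falloff, and bus and submix sends are disabled.
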SoundCue /Game/Sounds/Artemis/POI/Blimps/Spots/Spot_Blimp_Metal_Rubs_Distant_Cue
root=Modulator_0
nodes [
 {id=Modulator_0, children [Random_1], properties {pitchMin=0.9}}
 {
  id=Random_1
  children [WavePlayer_6, WavePlayer_7, WavePlayer_8, WavePlayer_9, WavePlayer_10, WavePlayer_11, WavePlayer_12, WavePlayer_13]
  properties {weights [1,1,1,1,1,1,1,1]}
 }
 {id=WavePlayer_10, properties {soundWaveAssetPtr=&/Spot_Blimp_Metal_Rub_02}}
 {id=WavePlayer_11, properties {soundWaveAssetPtr=&/Spot_Blimp_Metal_Rub_04}}
 {id=WavePlayer_12, properties {soundWaveAssetPtr=&/Spot_Blimp_Metal_Rub_06}}
 {id=WavePlayer_13, properties {soundWaveAssetPtr=&/Spot_Blimp_Metal_Rub_08}}
 {id=WavePlayer_6, properties {soundWaveAssetPtr=&/Spot_Blimp_Metal_Rub_01}}
 {id=WavePlayer_7, properties {soundWaveAssetPtr=&/Spot_Blimp_Metal_Rub_03}}
 {id=WavePlayer_8, properties {soundWaveAssetPtr=&/Spot_Blimp_Metal_Rub_05}}
 {id=WavePlayer_9, properties {soundWaveAssetPtr=&/Spot_Blimp_Metal_Rub_07}}
]
properties {
 attenuationSettings=/Game/Sounds/Attenuation/Fort_TileAmbient_OneShot_LessPositional_Log_Att
 bEnableBusSends=false
 bEnableSubmixSends=false
 maxDistance=6200
 soundClassObject=&Fort_Ambient_Outdoors_Athena
 volumeMultiplier=1.504
 attenuationOverrides {falloffDistance=6000, stereoSpread=10000, attenuationShapeExtents {x=2000}}
}
