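# Sound cue for a looping exterior ambience spot on blimp wings at the Artemis blimp POI.
# A short random delay starts an endlessly looping envelope around the looping wing wave, gently dipping and restoring volume over a 20-second cycle.
# It plays quietly in the gadgets sound class with capsule attenuation, and a concurrency override caps it at two voices.
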SoundCue /Game/Sounds/Artemis/POI/Blimps/Spots/Spot_Blimp_Wings_Exterior_Loop_Cue
root=Delay_0
nodes [
 {id=Delay_0, children [Enveloper_3], properties {delayMax=0.5, delayMin=0.1}}
 {
  id=Enveloper_3
  children [WavePlayer_1]
  properties {
   bLoop=true
   bLoopIndefinitely=true
   loopEnd=20
   pitchCurve {editorCurveData {keys [{time=0}]}}
   volumeCurve {editorCurveData {keys [
      {arriveTangent=-0.01, interpMode=RCIM_Cubic, leaveTangent=-0.01, time=0, value=1}
      <at 10 0.9 RCIM_Cubic>
      {arriveTangent=0.01, interpMode=RCIM_Cubic, leaveTangent=0.01, time=20, value=1}
     ]}}
  }
 }
 {id=WavePlayer_1, properties {bLooping=true, soundWaveAssetPtr=&/Ambience_Blimp_Wings_Exterior}}
]
properties {
 attenuationSettings=&/Spot_Blimp_Wings_Exterior_Capsule_Att
 bEnableBusSends=false
 bOverrideConcurrency=true
 maxDistance=6100
 soundClassObject=&Fort_Gadgets
 volumeMultiplier=0.3
 concurrencyOverrides {maxCount=2, voiceStealReleaseTime=2}
 concurrencySet [/Game/Sounds/Concurrency_Rules/Athena_Prop]
 preEffectBusSends [{}]
 soundSubmixSends [{minSendLevel=0, sendLevel=0, soundSubmix=&Fortnite_Reverb_Distant}]
}
