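# Looping outdoor ambient cue for a large roaring fire, filed under Blimps POI spot sounds and built from Superheated trap fire loops.
# Mixes close, medium and distant layers, each with its own attenuation asset, with the medium layer boosted.
# The close layer blends a rumble with one of two random inside-radius loops; the distant layer plays only in the first two quality-level slots.
# A root modulator randomizes pitch across the mix.
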SoundCue /Game/Sounds/Artemis/POI/Blimps/Spots/Spot_Fire_Roaring_Large_Loop_Cue
root=Modulator_3
nodes [
 {
  id=Attenuation_0
  children [Modulator_0]
  properties {attenuationSettings=/Game/Sounds/Apollo/Biomes/UFO_Crash_Sites/FireLoops/Fire_Large_Medium_Att}
 }
 {
  id=Attenuation_1
  children [Modulator_1]
  properties {attenuationSettings=/Game/Sounds/Apollo/Biomes/UFO_Crash_Sites/FireLoops/Fire_Large_Distant_Att}
 }
 {
  id=Attenuation_2
  children [Modulator_2]
  properties {attenuationSettings=/Game/Sounds/Apollo/Biomes/UFO_Crash_Sites/FireLoops/Fire_Large_Close_Att}
 }
 {id=Mixer_2, children [Attenuation_2, Attenuation_0, QualityLevel_0], properties {inputVolume [1,2,1]}}
 {id=Mixer_3, children [WavePlayer_3,Random_4], properties {inputVolume [1,1]}}
 {id=Modulator_0, children [WavePlayer_5]}
 {id=Modulator_1, children [WavePlayer_4]}
 {id=Modulator_2, children [Mixer_3]}
 {id=Modulator_3, children [Mixer_2], properties {pitchMax=1.1, pitchMin=0.9}}
 {id=QualityLevel_0, children [Attenuation_1, Attenuation_1, null, null]}
 {id=Random_4, children [WavePlayer_6,WavePlayer_7], properties {weights [1,1]}}
 {
  id=WavePlayer_3
  properties {bLooping=true, soundWaveAssetPtr=/Game/Sounds/Fort_GamePlay_Sounds/Superheated/SFX_Superheated_Trap_Fire_Rumble_Close_Loop_01}
 }
 {
  id=WavePlayer_4
  properties {bLooping=true, soundWaveAssetPtr=/Game/Sounds/Fort_GamePlay_Sounds/Superheated/SFX_Superheated_Trap_Fire_Rumble_Far_Loop_01}
 }
 {
  id=WavePlayer_5
  properties {bLooping=true, soundWaveAssetPtr=/Game/Sounds/Fort_GamePlay_Sounds/Superheated/SFX_Superheated_Trap_Fire_Rumble_Medium_Loop_01}
 }
 {
  id=WavePlayer_6
  properties {bLooping=true, soundWaveAssetPtr=/Game/Sounds/Fort_GamePlay_Sounds/Superheated/SFX_Superheated_Trap_Inside_Radius_Loop_01}
 }
 {
  id=WavePlayer_7
  properties {bLooping=true, soundWaveAssetPtr=/Game/Sounds/Fort_GamePlay_Sounds/Superheated/SFX_Superheated_Trap_Inside_Radius_Loop_02}
 }
]
properties {
 bOverrideConcurrency=true
 maxDistance=3500
 soundClassObject=&Fort_Ambient_Outdoors_Athena
 volumeMultiplier=0.752
 attenuationOverrides {
  bEnableFocusInterpolation=true
  bEnableListenerFocus=true
  distanceAlgorithm=Logarithmic
  falloffDistance=2500
  focusDistanceScale=0.9
  focusReleaseInterpSpeed=1.25
  nonFocusVolumeAttenuation=0.4
  stereoSpread=700
  attenuationShapeExtents {x=700}
  customAttenuationCurve {editorCurveData {keys [{arriveTangent=-2.3812, interpMode=RCIM_Cubic, leaveTangent=-2.3812, tangentMode=RCTM_User, time=0, value=1}, <at 1 0.01 RCIM_Cubic>]}}
 }
}
